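; Electric pole prop Blueprint for the NeonCity environment set, with no active gameplay logic in its event graph.
; Its construction script calls the parent, then loops over every StaticMeshComponent on the actor.
; Each mesh gets the FortBuildingMeshPhysics collision profile, Static mobility and component replication.
; A note in the script says this keeps player movement and interaction matched between client and server.
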(event EventBeginPlay
  (|CallParentFunction) ; This node is disabled and will not be called.\nDrag off pins to build functionality.
)

(event Collision|EventActorBeginOverlap
  (params (OtherActor :type Actor%20Object%20Reference)))

(event EventTick
  (params (DeltaSeconds :type Float%20%28single-precision%29)))

(fn ConstructionScript
  ;[1 1 1] Forcing proper collision settings, mobility settings, and replication to ensure player movement and interactions match on client and server
  (|CallParentFunction)
  (for _array_element (Actor|GetComponentsByClass "/Script/Engine.StaticMeshComponent")
    (Collision|SetCollisionProfileName _array_element "FortBuildingMeshPhysics" true :!class /Script/Engine.PrimitiveComponent)
    (Transformation|SetMobility _array_element "Static")
    (Components|SetIsReplicated _array_element true)))
